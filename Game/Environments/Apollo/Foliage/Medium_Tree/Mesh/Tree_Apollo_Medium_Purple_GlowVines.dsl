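; Construction script of the purple glow-vine medium Apollo tree Blueprint, which scatters instanced static meshes around stored transforms.
; After calling the parent script, it adds six instances to the InstancedStaticMesh for each entry in the Transforms array.
; Each instance gets a position offset rotated by its source transform, a random yaw combined with the source rotation, and random per-axis scale.
; All randomness draws from the Stream variable, allowing seeded, repeatable layouts.
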(fn ConstructionScript
  (|CallParentFunction)
  (for _array_element (Variables|Default|GetTransforms)
    (for _ (range 6)
      (Components|InstancedStaticMesh|AddInstance (Variables|Default|GetInstancedStaticMesh) (Math|Transform|MakeTransform (+ (.location _array_element) (Math|Vector|RotateVector (Math|Vector|MakeVector (Math|Random|RandomFloatinRangefromStream (Variables|Default|GetStream) (* (Variables|Default|GetPosOffsetMax) -1.0) (Variables|Default|GetPosOffsetMax)) (Math|Random|RandomFloatinRangefromStream (Variables|Default|GetStream) (* (Variables|Default|GetPosOffsetMax) -1.0) (Variables|Default|GetPosOffsetMax)) (Math|Random|RandomFloatinRangefromStream (Variables|Default|GetStream) 0.0 (Variables|Default|GetPos%20Z%20OffsetMax))) (.rotation _array_element))) (Math|Rotator|CombineRotators (.rotation _array_element) (Math|Rotator|MakeRotator 0.0 0.0 (Math|Random|RandomFloatinRangefromStream (Variables|Default|GetStream) 0.0 3.6e+02))) (Math|Vector|MakeVector (Math|Random|RandomFloatinRangefromStream (Variables|Default|GetStream) (Variables|Default|GetRand%20Scale%20min) (Variables|Default|GetRand%20Scale%20max)) (Math|Random|RandomFloatinRangefromStream (Variables|Default|GetStream) (Variables|Default|GetRand%20Scale%20min) (Variables|Default|GetRand%20Scale%20max)) (Math|Random|RandomFloatinRangefromStream (Variables|Default|GetStream) (Variables|Default|GetRand%20Scale%20min) (Variables|Default|GetRand%20Scale%20max)))) false))))
